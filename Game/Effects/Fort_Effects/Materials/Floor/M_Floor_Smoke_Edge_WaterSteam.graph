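# Translucent, unlit, two-sided effect material for water steam or smoke along a floor edge, judging by its name.
# Two panning inky-smoke samples, offset and speed-varied by vertex red, are warped by two rotating smoke textures except on mobile and low quality.
# Their sum, masked by a V gradient peaking mid-strip, is faded by scene depth and MF_SmoothLod camera distance for opacity.
# Emissive color blends two tints along squared V.
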Material /Game/Effects/Fort_Effects/Materials/Floor/M_Floor_Smoke_Edge_WaterSteam
graph {
 properties {blendMode=BLEND_Translucent, pixelDepthOffsetMode=PDOM_Legacy, shadingModel=MSM_Unlit, twoSided=true, usedShadingModels=Unlit}
 nodes [
  n1 = ComponentMask(None=TextureCoordinate()) {g=true}
  n2 = Multiply(A=ScalarParameter() {defaultValue=1, parameterName=Opacity}, B=Multiply(A=@n1, B=OneMinus(None=@n1)))
  n3 = VertexColor()
  n4 = AppendVector(A=@n3.R, B=@n3.R)
  n5 = TextureCoordinate()
  n6 = AppendVector(A=ScalarParameter() {parameterName=TileX}, B=ScalarParameter() {parameterName=TileY})
  n7 = VertexColor()
  n8 = AppendVector(A=@n7.R, B=@n7.R)
  n9 = ScalarParameter() {defaultValue=0.1, parameterName=SpeedMult}
  n10 = TextureCoordinate()
  n11 = TextureSample(UVs=Rotator(Coordinate=Multiply(A=@n10, B=@n6)) {centerX=1, speed=0.01}) {samplerType=SAMPLERTYPE_LinearColor, texture=/Game/Global/Textures/Default/Tileable_Patterns/T_Tile_Smoke_Small}
  n12 = TextureCoordinate()
  n13 = TextureSample(UVs=Rotator(Coordinate=Multiply(A=@n12, B=@n6)) {centerY=1, speed=-0.0113}) {samplerType=SAMPLERTYPE_LinearColor, texture=&T_Tile_Smoke_Small}
  n14 = Multiply(
   A=Add(A=ComponentMask(None=@n11.RGB) {b=true, g=true, r=true}, B=ComponentMask(None=@n13.RGB) {b=true, g=true, r=true})
   B=ScalarParameter() {defaultValue=0.1, parameterName=BaseDistort}
  )
  n15 = Constant()
  n16 = ShadingPathSwitch(Default=QualitySwitch(Default=ComponentMask(None=@n14) {g=true, r=true}, Low=@n15), Mobile=@n15) {desc=Migrated from FeatureLevelSwitch}
  n17 = Add(A=Panner(Coordinate=Multiply(A=@n5, B=@n6), Speed=Multiply(A=@n8, B=@n9)) {speedX=0.02, speedY=0.15}, B=@n16)
  n18 = VertexColor()
  n19 = AppendVector(A=@n18.R, B=@n18.R)
  n20 = ScalarParameter() {defaultValue=0.1, parameterName=SpeedMult_02}
  n21 = Add(A=Panner(Coordinate=Multiply(A=@n5, B=@n6), Speed=Multiply(A=@n8, B=@n20)) {speedX=-0.02, speedY=0.1}, B=@n16)
  n22 = Add(
   A=TextureSample(UVs=Add(A=@n4, B=@n17)) {texture=&T_Inky_Smoke_Tile}.G
   B=TextureSample(UVs=Add(A=@n19, B=@n21)) {texture=&T_Inky_Smoke_Tile}.G
  )
  n23 = ScalarParameter() {parameterName=FadeDistance}
  Multiply(
   A=Multiply(A=Multiply(A=@n2, B=@n22), B=DepthFade(FadeDistance=@n23))
   B=MaterialFunctionCall!(
    NearFadeLength=ScalarParameter() {defaultValue=21000, parameterName="NearFadeLength (S)"}
    NearFadeOffset=ScalarParameter() {defaultValue=20000, parameterName="NearFadeOffset (S)"}
    NearFadeCurve
    NearFadeStartingScale
    NearFadeEndingScale
    FarFadeLength=ScalarParameter() {defaultValue=25000, parameterName="FarFadeLength (S)"}
    FarFadeOffset=ScalarParameter() {defaultValue=125000, parameterName="FarFadeOffset (S)"}
    FarFadeCurve
    FarFadeEndingScale
    Opacity Near Distance Bias
    Opacity Far Distance Bias=ScalarParameter() {defaultValue=-1, parameterName="Opacity Far Distance Bias (S)"}
   ) {materialFunction=/Game/Effects/Fort_Effects/Materials/_Functions/MF_SmoothLod}.Opacity
  )
  n24 = ComponentMask(None=TextureCoordinate()) {g=true}
  Multiply(
   A=ScalarParameter() {defaultValue=5, parameterName=EmmsiveMult}
   B=LinearInterpolate(
    A=VectorParameter() {parameterName=Color_1, defaultValue=<rgba 0.2882 0.3853 0.624 1>}.RGB
    B=VectorParameter() {parameterName=Color, defaultValue=<rgba 0.4969 0.552 0.6105 1>}.RGB
    Alpha=Multiply(A=@n24, B=@n24)
   )
  )
 ]
}
