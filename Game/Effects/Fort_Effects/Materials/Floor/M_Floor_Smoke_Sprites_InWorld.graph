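# Translucent unlit material for large in-world floor smoke sprites, with a terrain height check labelled for lava.
# Sprite offsets are sized by vertex color and distance LOD, rotated, and optionally faded out where a volcano heightmap rises above the actor.
# Random 8x8 sub-UV smoke, distorted and masked by panning noise, drives opacity with distance and depth fades.
# Color blends by view height, tinted by time of day and Schlick-phase sun scattering.
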Material /Game/Effects/Fort_Effects/Materials/Floor/M_Floor_Smoke_Sprites_InWorld
graph {
 properties {blendMode=BLEND_Translucent, pixelDepthOffsetMode=PDOM_Legacy, shadingModel=MSM_Unlit, usedShadingModels=Unlit}
 nodes [
  Add()
  Constant() {r=1}
  n1 = Power(
   Base=MaterialFunctionCall!(
    NearFadeLength=ScalarParameter() {defaultValue=15000, group=Scaler, parameterName="NearFadeLength (S)"}
    NearFadeOffset=ScalarParameter() {defaultValue=10000, group=Scaler, parameterName="NearFadeOffset (S)"}
    NearFadeCurve
    NearFadeStartingScale
    NearFadeEndingScale
    FarFadeLength=ScalarParameter() {defaultValue=15000, group=Scaler, parameterName="FarFadeLength (S)"}
    FarFadeOffset=ScalarParameter() {defaultValue=75000, group=Scaler, parameterName="FarFadeOffset (S)"}
    FarFadeCurve
    FarFadeEndingScale
    Opacity Near Distance Bias=ScalarParameter() {group=Scaler, parameterName="Opacity Near Distance Bias Opacity (S)"}
    Opacity Far Distance Bias=ScalarParameter() {group=Scaler, parameterName="Opacity Far Distance Bias Opacity (S)"}
   ) {materialFunction=/Game/Effects/Fort_Effects/Materials/_Functions/MF_SmoothLod}.Opacity
   Exp=ScalarParameter() {defaultValue=4, group=Opacity, parameterName=OpactiyDistScalePower}
  )
  n2 = ScalarParameter() {defaultValue=8, group=Texture values, parameterName=RowsAndColumns}
  n3 = VertexColor()
  n4 = Divide(A=Ceil(None=Multiply(A=@n3.R) {constB=64})) {constB=64}
  n5 = Clamp(None=MaterialFunctionCall!(UseTexCoord 1, Rows=@n2, Columns=@n2, Particle Random Value=Clamp(None=@n4)) {materialFunction=/Game/Effects/Fort_Effects/Materials/_Functions/MF_RandomSubUV_GPU_Sprites}.Result)
  n6 = VertexColor()
  n7 = TextureCoordinate()
  n8 = AppendVector(A=ScalarParameter() {group=Texture values, parameterName=TileX}, B=ScalarParameter() {group=Texture values, parameterName=TileY})
  n9 = VertexColor()
  n10 = AppendVector(A=@n9.R, B=@n9.R)
  n11 = Panner(
   Coordinate=Multiply(A=@n7, B=@n8)
   Speed=Multiply(A=@n10, B=ScalarParameter() {defaultValue=0.1, group=Texture values, parameterName=SpeedMult})
  ) {speedX=0.02, speedY=0.15}
  n12 = TextureCoordinate()
  n13 = TextureSample(UVs=Rotator(Coordinate=Multiply(A=@n12, B=@n8)) {centerX=1, speed=0.0312}) {samplerType=SAMPLERTYPE_LinearColor, texture=/Game/Global/Textures/Default/Tileable_Patterns/T_Tile_Smoke_Small}
  n14 = TextureCoordinate()
  n15 = TextureSample(UVs=Rotator(Coordinate=Multiply(A=@n14, B=@n8)) {centerY=1, speed=-0.0312}) {samplerType=SAMPLERTYPE_LinearColor, texture=&T_Tile_Smoke_Small}
  n16 = Multiply(
   A=Add(A=ComponentMask(None=@n13.RGB) {b=true, g=true, r=true}, B=ComponentMask(None=@n15.RGB) {b=true, g=true, r=true})
   B=ScalarParameter() {defaultValue=0.1, group=Texture values, parameterName=BaseDistort}
  )
  n17 = Constant()
  n18 = ShadingPathSwitch(Default=QualitySwitch(Default=ComponentMask(None=@n16) {g=true, r=true}, Low=@n17), Mobile=@n17) {desc=Migrated from FeatureLevelSwitch}
  n19 = TextureSample(UVs=Add(A=AppendVector(A=@n6.R, B=@n6.R), B=Add(A=@n11, B=@n18))) {texture=&T_Inky_Smoke_Tile}
  n20 = VertexColor()
  n21 = Panner(
   Coordinate=Multiply(A=@n7, B=@n8)
   Speed=Multiply(A=@n10, B=ScalarParameter() {defaultValue=0.1, group=Texture values, parameterName=SpeedMult_02})
  ) {speedX=-0.02, speedY=0.1}
  n22 = TextureSample(UVs=Add(A=AppendVector(A=@n20.R, B=@n20.R), B=Add(A=@n21, B=@n18))) {texture=&T_Inky_Smoke_Tile}
  n23 = Multiply(A=AppendVector(A=@n19.G, B=@n22.G), B=ScalarParameter() {defaultValue=0.01, group=Texture values, parameterName=BaseDistortUV})
  n24 = Constant()
  n25 = ShadingPathSwitch(Default=QualitySwitch(Default=@n23, Low=@n24, High=@n23, Medium=@n24), Mobile=@n24) {desc=Migrated from FeatureLevelSwitch}
  n26 = StaticSwitchParameter(True=Add(A=@n5, B=ShadingPathSwitch(Default=@n25, Forward=@n24)), False=@n5) {defaultValue=true, parameterName=UseUVDistort?}
  n27 = Multiply(
   A=Multiply(
    A=TextureSample(UVs=@n26) {samplerType=SAMPLERTYPE_LinearColor, texture=/Game/Effects/Fort_Effects/Textures/Smoke/T_Smoke_8x8_01}.G
    B=Add(A=@n19.G, B=@n22.G)
   )
   B=VertexColor().B
  )
  n28 = DepthFade(FadeDistance=ScalarParameter() {defaultValue=5000, group=Opacity, parameterName=FadeDistance})
  n29 = Constant() {r=1}
  Multiply(
   A=Multiply(A=Multiply(A=@n1, B=@n27), B=ScalarParameter() {parameterName=OpacityMult})
   B=ShadingPathSwitch(Default=QualitySwitch(Default=@n28, Low=@n29), Mobile=@n29) {desc=Migrated from FeatureLevelSwitch}
  )
  ScalarParameter() {defaultValue=-1, group=Scaler, parameterName="Opacity Far Distance Bias (S)"}
  n30 = Constant3Vector()
  n31 = VertexNormalWS()
  n32 = MaterialFunctionCall(
   NearFadeLength=ScalarParameter() {defaultValue=15000, group=Scaler, parameterName="NearFadeLength (S)"}
   NearFadeOffset=ScalarParameter() {defaultValue=10000, group=Scaler, parameterName="NearFadeOffset (S)"}
   FarFadeLength=ScalarParameter() {defaultValue=15000, group=Scaler, parameterName="FarFadeLength (S)"}
   FarFadeOffset=ScalarParameter() {defaultValue=75000, group=Scaler, parameterName="FarFadeOffset (S)"}
   Opacity Near Distance Bias=ScalarParameter() {group=Scaler, parameterName="Opacity Near Distance Bias (S)"}
   Opacity Far Distance Bias=ScalarParameter() {defaultValue=-1, group=Scaler, parameterName="Opacity Far Distance Bias (S)"}
  ) {materialFunction=&MF_SmoothLod}
  n33 = ScalarParameter() {parameterName=Min}
  n34 = ScalarParameter() {defaultValue=6000, group=SpriteSize, parameterName=SpriteSizeMin}
  n35 = AppendVector(A=@n34, B=@n34)
  n36 = ScalarParameter() {defaultValue=10000, group=SpriteSize, parameterName=SpriteSizeMax}
  n37 = AppendVector(A=@n36, B=@n36)
  n38 = VertexColor()
  n39 = MaterialFunctionCall!(
   Center Location=WorldPosition().XYZ
   0-1 UVs
   XY scale=Multiply(A=Clamp(None=@n32.Opacity, Min=@n33), B=LinearInterpolate(A=@n35, B=@n37, Alpha=@n38.G))
   OverridePivot
   "Normals (Optional)"
  ) {materialFunction=&Sprite}
  n40 = VertexColor()
  n41 = RotateAboutAxis(
   NormalizedRotationAxis=LinearInterpolate(A=Constant3Vector(), B=Constant3Vector() {constant=<rgba 1 1 1 0>}, Alpha=@n40.R)
   RotationAngle=Multiply(A=Time(), B=@n40.G)
   PivotPoint=WorldPosition().XYZ
   Position=WorldPosition().XYZ
  )
  n42 = TextureSampleParameter2D(UVs=TextureCoordinate() {coordinateIndex=1}) {
   group=LavaHeightCheck
   parameterName=TerrainHightMap
   samplerType=SAMPLERTYPE_LinearGrayscale
   texture=/Game/Environments/Sets/Vulcano/Textures/T_TerrainHeightMap
  }
  n43 = LinearInterpolate(
   A=ScalarParameter() {defaultValue=-2500, group=LavaHeightCheck, parameterName=TerrainHeight Mininum}
   B=ScalarParameter() {defaultValue=11082.7412, group=LavaHeightCheck, parameterName=TerrainHeightMax}
   Alpha=@n42.R
  )
  n44 = ComponentMask(None=ActorPositionWS()) {b=true}
  n45 = OneMinus(None=Divide(A=Subtract(A=@n43, B=@n44), B=ScalarParameter() {defaultValue=1500, group=LavaHeightCheck, parameterName=TerrainHeightCheckRange}))
  StaticSwitchParameter(
   True=LinearInterpolate(A=Multiply(A=@n30, B=@n31), B=Add(A=@n39.World position offset, B=@n41), Alpha=Clamp(None=@n45))
   False=@n39.World position offset
  ) {defaultValue=true, parameterName=FadeBasedOnHeightMap?}
  n46 = VectorParameter() {parameterName=Color_Top, defaultValue=<rgba 3.75 4.1102 5 1>}
  n47 = Multiply(
   A=VectorParameter() {parameterName=Color_Bottom, defaultValue=<rgba 1 0.25 0 0>}.RGB
   B=MaterialFunctionCall(
    Value_Morning=ScalarParameter() {defaultValue=0.5, group=TOD, parameterName="Value_Morning (S)"}
    Value_Day=ScalarParameter() {defaultValue=0.5, group=TOD, parameterName="Value_Day (S)"}
    Value_Evening=ScalarParameter() {defaultValue=0.25, group=TOD, parameterName="Value_Evening (S)"}
    Value_Night=ScalarParameter() {defaultValue=0.1, group=TOD, parameterName="Value_Night (S)"}
   ) {materialFunction=&Scalar_Time_of_Day}.Result
  )
  n48 = ComponentMask(None=CameraVectorWS()) {b=true}
  n49 = ScalarParameter() {defaultValue=0.05, parameterName=offset}
  n50 = Clamp(None=Divide(A=Add(A=@n48, B=@n49), B=ScalarParameter() {defaultValue=0.2, parameterName=Div}))
  n51 = CollectionParameter() {
   collection=&FortniteMaterialParameters
   parameterName=SunlightDirectionalVector
  }
  MaterialExpressionDotProduct_0 = DotProduct(A=ComponentMask(None=@n51) {b=true, g=true, r=true}, B=CameraVectorWS())
  MaterialExpressionScalarParameter_9 = ScalarParameter() {defaultValue=0.4, parameterName=Anisotropy}
  n52 = Custom!(CosTheta=@MaterialExpressionDotProduct_0, k=@MaterialExpressionScalarParameter_9) {
   code="\tfloat Inner = (1 + k * CosTheta);\r\n\treturn (1 - k * k) / (4 * PI * Inner * Inner);\r\n\r\n"
   description=Schlick Phase
   inputs [
    {
     inputName=CosTheta
     input {expression=&self:MaterialExpressionDotProduct_0}
    }
    {inputName=k, input {expression=&self:MaterialExpressionScalarParameter_9}}
   ]
  }
  n53 = MaterialFunctionCall(
   Color_Morning=VectorParameter() {group=TOD, parameterName="Color_Morning (V3)", defaultValue=<rgba 1 0.7406 0.5664 0>}.RGB
   Color_Day=VectorParameter() {group=TOD, parameterName="Color_Day (V3)", defaultValue=<rgba 0.75 0.3433 0.3359 0>}.RGB
   Color_Evening=VectorParameter() {group=TOD, parameterName="Color_Evening (V3)", defaultValue=<rgba 1 0.6005 0.2592 0>}.RGB
   Color_Night=VectorParameter() {group=TOD, parameterName="Color_Night (V3)", defaultValue=<rgba 1 0.4906 0.3 0>}.RGB
  ) {materialFunction=&Color_Time_of_Day}
  n54 = Multiply(A=Multiply(A=@n52, B=ScalarParameter() {defaultValue=28, parameterName=ScaleCloudColor}), B=Multiply(A=@n53.Result) {constB=2})
  StaticSwitchParameter(
   True=Multiply(A=LinearInterpolate(A=@n46.RGB, B=@n47, Alpha=@n50), B=Multiply(A=@n54, B=@n53.Result))
   False=VectorParameter() {parameterName=Color_Whole, defaultValue=<rgba 1 0.25 0 0>}.RGB
  ) {defaultValue=true, parameterName=UseTOD?}
  TextureCoordinate()
  VertexColor()
 ]
}
